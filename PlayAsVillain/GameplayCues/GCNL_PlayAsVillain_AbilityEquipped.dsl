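; Looping GameplayCue for the PlayAsVillain equipped ability that shows or hides a particle effect on the player pawn.
; The effect starts hidden, appears when the Power cooldown tag is removed, and hides when UpgradedOvershield.Active is added.
; A parallel tag watch tracks cloaking, hiding the effect while the pawn is invisible and restoring it afterward if still equipped.
; ToggleFX fades by playing or reversing a timeline that drives the HideFX particle parameter.
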(event EventOnApplication(EitherVFX-preferthisone)
  (params (MyTarget :type Actor%20Object%20Reference) (Parameters :type Gameplay%20Cue%20Parameters%20Structure%20%28by%20ref%29) (ParticleComponents :type Array%20of%20FXSystem%20Component%20Object%20References) (AudioComponents :type Array%20of%20Audio%20Component%20Object%20References) (BurstCameraShakeInstance :type Camera%20Shake%20Base%20Object%20Reference) (BurstDecalInstance :type Decal%20Actor%20Object%20Reference))
  (|CallParentFunction MyTarget Parameters ParticleComponents AudioComponents BurstCameraShakeInstance BurstDecalInstance))

(event EventOnLoopingStart(EitherVFX-preferthisversion)
  (params (MyTarget :type Actor%20Object%20Reference) (Parameters :type Gameplay%20Cue%20Parameters%20Structure%20%28by%20ref%29) (ParticleComponents :type Array%20of%20FXSystem%20Component%20Object%20References) (AudioComponents :type Array%20of%20Audio%20Component%20Object%20References))
  (|CallParentFunction MyTarget Parameters ParticleComponents AudioComponents)
  (bind _asfort_player_pawn_athena (Utilities|Casting|CastToFortPlayerPawnAthena MyTarget))
  (Variables|Default|SetPawnRef _asfort_player_pawn_athena)
  (bind _output_get (Variables|Default|SetEquippedFX (Utilities|Array|Get(acopy) ParticleComponents)))
  (Variables|Default|SetisAbilityEquipped false)
  (Effects|Components|ParticleSystem|SetFloatParameter _output_get "HideFX" 1.0)
  (Utilities|FlowControl|Sequence
    (:then_0
      ;[0.136 0.5981 0.7917 0.5] Check if Pawn is Invisible
      (Ability|Tasks|WaitGameplayTagRemoveFromActor (Variables|Default|GetPawnRef) "(TagName=\"Athena.Gameplay.Status.Cloaked\")"
        (:then
          (Ability|Tasks|WaitGameplayTagAddToActor (Variables|Default|GetPawnRef) "(TagName=\"Athena.Gameplay.Status.Cloaked\")"
            (:Added
              (Variables|Default|SetisPawnInvisible true)
              (if (Variables|Default|GetisAbilityEquipped)
                (CallFunction|ToggleFX false)))))
        (:Removed
          (Variables|Default|SetisPawnInvisible false)
          (if (Variables|Default|GetisAbilityEquipped)
            (CallFunction|ToggleFX true)))))
    (:then_1
      ;[0.8646 0.6419 0.2026 0.5] Check if Ability item is Equipped
      (Ability|Tasks|WaitGameplayTagRemoveFromActor (Variables|Default|GetPawnRef) "(TagName=\"Asset.Cooldown.PlayAsVillain.Power\")"
        (:Removed
          (Ability|Tasks|WaitGameplayTagAddToActor (Variables|Default|GetPawnRef) "(TagName=\"PlayAsVillain.UpgradedOvershield.Active\")"
            (:then
              (Variables|Default|SetisAbilityEquipped true)
              (if (not (Variables|Default|GetisPawnInvisible))
                (CallFunction|ToggleFX true)))
            (:Added
              (Variables|Default|SetisAbilityEquipped false)
              (if (not (Variables|Default|GetisPawnInvisible))
                (CallFunction|ToggleFX false)))))))))

(event Custom|ToggleFX
  (params (ActivateFX :type Boolean))
  (if ActivateFX
    (bind (_direction _fadealpha) (Timeline|Fade%20Timeline ReverseFromEnd)
      (:Update
        (Effects|Components|ParticleSystem|SetFloatParameter (Variables|Default|GetEquippedFX) "HideFX" _fadealpha)))
    (else
      (Timeline|Fade%20Timeline PlayFromStart))))

(fn ConstructionScript)
