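; Parent looping gameplay cue for the Play As Villain player aura, driving the aura particle effect on the target pawn.
; On loop start it caches the pawn and first particle component, flags whether the pawn is locally viewed, and passes the effect level to Niagara as BossIndex.
; Async waits on the Cloaked tag play or reverse a fade timeline that drives the aura's Invisibility parameter.
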(event EventBeginPlay)

(event Collision|EventActorBeginOverlap
  (params (OtherActor :type Actor%20Object%20Reference)))

(event EventTick
  (params (DeltaSeconds :type Float%20%28single-precision%29)))

(event EventOnLoopingStart(EitherVFX-preferthisversion)
  (params (MyTarget :type Actor%20Object%20Reference) (Parameters :type Gameplay%20Cue%20Parameters%20Structure%20%28by%20ref%29) (ParticleComponents :type Array%20of%20FXSystem%20Component%20Object%20References) (AudioComponents :type Array%20of%20Audio%20Component%20Object%20References))
  (bind (_normalizedmagnitude _rawmagnitude _effectcontext _matchedtagname _originaltag _aggregatedsourcetags _aggregatedtargettags _location _normal _instigator _effectcauser _sourceobject _physicalmaterial _gameplayeffectlevel) (Ability|GameplayCue|BreakGameplayCueParameters Parameters))
  (bind _output (Utilities|Array|Get(acopy) ParticleComponents))
  (|CallParentFunction MyTarget Parameters ParticleComponents AudioComponents)
  (bind _asfort_player_pawn_athena (Utilities|Casting|CastToFortPlayerPawnAthena MyTarget))
  (bind _output_get (Variables|Default|SetPawnRef _asfort_player_pawn_athena))
  (bind _output_get_2 (Variables|Default|SetAuraFX _output))
  (Effects|Components|ParticleSystem|SetBoolParameter _output_get_2 "IsLocal" (Pawn|IsLocallyViewed _output_get))
  (bind _asniagara_particle_system_component (Utilities|Casting|CastToNiagaraComponent _output)
    (:then
      (Niagara|SetNiagaraVariable(Int32) _asniagara_particle_system_component "BossIndex" _gameplayeffectlevel))
    (:CastFailed
      _))
  (Ability|Tasks|WaitGameplayTagRemoveFromActor (Variables|Default|GetPawnRef) "(TagName=\"Athena.Gameplay.Status.Cloaked\")"
    (:then
      (Ability|Tasks|WaitGameplayTagAddToActor (Variables|Default|GetPawnRef) "(TagName=\"Athena.Gameplay.Status.Cloaked\")"
        (:Added
          (bind (_direction _fadealpha) (Timeline|FadeTimeline PlayFromStart)
            (:Update
              (Effects|Components|ParticleSystem|SetFloatParameter (Variables|Default|GetAuraFX) "Invisibility" _fadealpha))))))
    (:Removed
      (Timeline|FadeTimeline ReverseFromEnd))))

(fn ConstructionScript)
